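# Physics asset for the DieselPunk 02 medium male player body, covering the upper spine, both clavicles and two dynamic flap bones on the left side.
# Box bodies on the two flaps hang from the left clavicle through constraints with limited swing and twist.
# Capsules cover spine_05 and the clavicles, which join the spine with free angular motion, and the left clavicle capsule is much smaller than the right.
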PhysicsAsset /BRCosmetics/Characters/Player/Male/Medium/Bodies/M_MED_DieselPunk_02/Meshes/M_MED_Dieselpunk_02_Physics
bodies {
 clavicle_l [{shapeType=Capsule, length=1, radius=1.3236}]
 clavicle_r [{shapeType=Capsule, length=14.6718, radius=3.1213, center=<-9.5697 -0.2015 0.1698>, rotation=<rot 0 -176.5466 -79.5102>}]
 dyn_flapA_l [{shapeType=Box, extentX=3.1498, extentY=1.1734, extentZ=5.9325, center=<-0.9111 0.118 -2.0436>, rotation=<rot -7.1457 -90 12.0395>}]
 dyn_flapB_l [{shapeType=Box, extentX=3.1636, extentY=1.3654, extentZ=5.5755, center=<-1.1892 -0.0866 -2.2205>, rotation=<rot 3.0315 -89.9999 22.0062>}]
 spine_05 [{shapeType=Capsule, length=13.4871, radius=10.5213}]
}
constraints [
 {bone1Name=dyn_flapA_l, bone2Name=clavicle_l, swing1LimitDegrees=30, swing2LimitDegrees=45, twistLimitDegrees=26.1429}
 {bone1Name=dyn_flapB_l, bone2Name=clavicle_l, swing1LimitDegrees=45, swing2LimitDegrees=45, twistLimitDegrees=15}
 {
  bone1Name=clavicle_l
  bone2Name=spine_05
  swing1LimitDegrees=45
  swing1Motion=Free
  swing2LimitDegrees=45
  swing2Motion=Free
  twistLimitDegrees=45
  twistMotion=Free
 }
 {
  bone1Name=clavicle_r
  bone2Name=spine_05
  swing1LimitDegrees=45
  swing1Motion=Free
  swing2LimitDegrees=45
  swing2Motion=Free
  twistLimitDegrees=45
  twistMotion=Free
 }
]
