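# Animation Blueprint for the male medium DieselPunk 02 body that corrects its shoulder pad and clothing bones as the character moves.
# RBF Pose Drivers read spine, clavicle and upper-arm rotations from a cached pose to drive pads via pose assets.
# Weapon booleans (rocket launcher, chaingun, rifle, pistol) select left-pad drivers, while skydiving toggles a right-pad driver.
# Weighted multi-bone constraints then position zipper, pad and flap bones.
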AnimBlueprint /BRCosmetics/Characters/Player/Male/Medium/Bodies/M_MED_DieselPunk_02/Meshes/M_MED_Dieselpunk_02_AnimBP
target_skeleton=&/M_MED_Dieselpunk_02_Skeleton
graphs [
 AnimGraph {nodes [
   n1 = Pose Driver - Source: clavicle_l!(SourcePose:Pose Link Structure=Use cached pose 'AnimCache'()) -> (Pose:Pose Link Structure) {
    is=title AnimGraphNode_PoseDriver
    poseAsset=&/M_MED_DieselPunk_02_Arms_PoseAsset
    onlyDriveBones [{boneName=driver_pad_l}, {boneName=dyn_pad_bck}]
    poseTargets [
     {
      bApplyCustomCurve=false
      bIsHidden=false
      distanceMethod=DefaultMethod
      drivenName=bind
      functionType=DefaultFunction
      targetScale=0.6488
      boneTransforms [
       {targetRotation=<rot -85.1913 -138.5624 143.0618>, targetTranslation=<4.3575 -0.1466 -2.3223>}
       {targetRotation=<rot -64.7854 -3.089 6.0959>, targetTranslation=<15.5111 0 0>}
      ]
      customCurve {defaultValue=3.4028234663852886e+38, postInfinityExtrap=RCCE_Constant, preInfinityExtrap=RCCE_Constant, keys []}
      targetRotation=<rot 0 0 0>
     }
     {
      bApplyCustomCurve=false
      bIsHidden=false
      distanceMethod=DefaultMethod
      drivenName=pistol_fwd
      functionType=DefaultFunction
      targetScale=0.5513
      boneTransforms [
       {targetRotation=<rot -80.1647 47.1898 -43.5433>, targetTranslation=<4.3575 -0.1466 -2.3223>}
       {targetRotation=<rot -27.4019 74.8174 -63.7631>, targetTranslation=<15.5111 0 0>}
      ]
      customCurve {defaultValue=3.4028234663852886e+38, postInfinityExtrap=RCCE_Constant, preInfinityExtrap=RCCE_Constant, keys []}
      targetRotation=<rot 0 0 0>
     }
     {
      bApplyCustomCurve=false
      bIsHidden=false
      distanceMethod=DefaultMethod
      drivenName=pistol_up
      functionType=DefaultFunction
      targetScale=1
      boneTransforms [
       {targetRotation=<rot -45.5272 7.2881 -5.3683>, targetTranslation=<4.3575 -0.1466 -2.3223>}
       {targetRotation=<rot 9.8617 54.2947 -96.1665>, targetTranslation=<15.5111 0 0>}
      ]
      customCurve {defaultValue=3.4028234663852886e+38, postInfinityExtrap=RCCE_Constant, preInfinityExtrap=RCCE_Constant, keys []}
      targetRotation=<rot 0 0 0>
     }
     {
      bApplyCustomCurve=false
      bIsHidden=false
      distanceMethod=DefaultMethod
      drivenName=pistol_dwn
      functionType=DefaultFunction
      targetScale=0.5702
      boneTransforms [
       {targetRotation=<rot -81.543 57.4431 -54.3032>, targetTranslation=<4.3575 -0.1466 -2.3223>}
       {targetRotation=<rot -70.1236 89.2767 -56.32>, targetTranslation=<15.5111 0 0>}
      ]
      customCurve {defaultValue=3.4028234663852886e+38, postInfinityExtrap=RCCE_Constant, preInfinityExtrap=RCCE_Constant, keys []}
      targetRotation=<rot 0 0 0>
     }
     {
      bApplyCustomCurve=false
      bIsHidden=false
      distanceMethod=DefaultMethod
      drivenName=pistol_bck
      functionType=DefaultFunction
      targetScale=0.5513
      boneTransforms [
       {targetRotation=<rot -82.1915 81.681 -80.2223>, targetTranslation=<4.3575 -0.1466 -2.3223>}
       {targetRotation=<rot -36.6606 93.9397 -43.0028>, targetTranslation=<15.5111 0 0>}
      ]
      customCurve {defaultValue=3.4028234663852886e+38, postInfinityExtrap=RCCE_Constant, preInfinityExtrap=RCCE_Constant, keys []}
      targetRotation=<rot 0 0 0>
     }
    ]
    rBFParams {distanceMethod=Quaternion, radius=33.1614}
    sourceBones [{boneName=clavicle_l}, {boneName=upperarm_l}]
   }
   n2 = Blend Poses by bool(BlendPose_0=@n1, bActiveValue=Get pistol_active() -> (pistol_active:bool)) {blendType=Linear}
   n3 = Pose Driver - Source: spine_05!(SourcePose:Pose Link Structure=Use cached pose 'AnimCache'()) -> (Pose:Pose Link Structure) {
    is=title AnimGraphNode_PoseDriver
    poseAsset=&M_MED_DieselPunk_02_Arms_PoseAsset
    onlyDriveBones [{boneName=driver_pad_l}, {boneName=dyn_pad_bck}]
    poseTargets [
     {
      bApplyCustomCurve=false
      bIsHidden=false
      distanceMethod=DefaultMethod
      drivenName=bind
      functionType=DefaultFunction
      targetScale=0.613
      boneTransforms [
       {targetRotation=<rot 0 0.1712 0>, targetTranslation=<10.0146 0 0>}
       {targetRotation=<rot -85.1913 -138.5624 143.0618>, targetTranslation=<4.3575 -0.1466 -2.3223>}
       {targetRotation=<rot -64.7854 -3.089 6.0959>, targetTranslation=<15.5111 0 0>}
      ]
      customCurve {defaultValue=3.4028234663852886e+38, postInfinityExtrap=RCCE_Constant, preInfinityExtrap=RCCE_Constant, keys []}
      targetRotation=<rot 0 0 0>
     }
     {
      bApplyCustomCurve=false
      bIsHidden=false
      distanceMethod=DefaultMethod
      drivenName=rifle_fwd
      functionType=DefaultFunction
      targetScale=1
      boneTransforms [
       {targetRotation=<rot -1.8763 -1.5677 -6.688>, targetTranslation=<10.0146 0 0>}
       {targetRotation=<rot -82.3641 52.539 -51.6926>, targetTranslation=<4.3575 -0.1466 -2.3223>}
       {targetRotation=<rot -37.0534 63.6677 -82.6126>, targetTranslation=<15.5111 0 0>}
      ]
      customCurve {defaultValue=3.4028234663852886e+38, postInfinityExtrap=RCCE_Constant, preInfinityExtrap=RCCE_Constant, keys []}
      targetRotation=<rot 0 0 0>
     }
     {
      bApplyCustomCurve=false
      bIsHidden=false
      distanceMethod=DefaultMethod
      drivenName=rifle_up
      functionType=DefaultFunction
      targetScale=1
      boneTransforms [
       {targetRotation=<rot -2.8104 -0.2929 -6.8311>, targetTranslation=<10.0146 0 0>}
       {targetRotation=<rot -51.6841 10.4523 -9.0777>, targetTranslation=<4.3575 -0.1466 -2.3223>}
       {targetRotation=<rot -7.9729 51.1321 -83.85>, targetTranslation=<15.5111 0 0>}
      ]
      customCurve {defaultValue=3.4028234663852886e+38, postInfinityExtrap=RCCE_Constant, preInfinityExtrap=RCCE_Constant, keys []}
      targetRotation=<rot 0 0 0>
     }
     {
      bApplyCustomCurve=false
      bIsHidden=false
      distanceMethod=DefaultMethod
      drivenName=rifle_dwn
      functionType=DefaultFunction
      targetScale=0.613
      boneTransforms [
       {targetRotation=<rot -2.6929 -0.7332 -8.0647>, targetTranslation=<10.0146 0 0>}
       {targetRotation=<rot -82.4636 143.5654 -137.6768>, targetTranslation=<4.3575 -0.1466 -2.3223>}
       {targetRotation=<rot -67.2141 49.9006 -28.3212>, targetTranslation=<15.5111 0 0>}
      ]
      customCurve {defaultValue=3.4028234663852886e+38, postInfinityExtrap=RCCE_Constant, preInfinityExtrap=RCCE_Constant, keys []}
      targetRotation=<rot 0 0 0>
     }
    ]
    rBFParams {distanceMethod=Quaternion, function=Linear, radius=22.9601}
    sourceBones [{boneName=spine_05}, {boneName=clavicle_l}, {boneName=upperarm_l}]
   }
   n4 = Get rifle_active() -> (rifle_active:bool)
   n5 = Pose Driver - Source: spine_05(SourcePose=Use cached pose 'AnimCache'()) {
    poseAsset=&M_MED_DieselPunk_02_Arms_PoseAsset
    onlyDriveBones [{boneName=driver_pad_l}, {boneName=dyn_pad_bck}]
    poseTargets [
     {
      bApplyCustomCurve=false
      bIsHidden=false
      distanceMethod=DefaultMethod
      drivenName=bind
      functionType=DefaultFunction
      targetScale=0.5964
      boneTransforms [
       {targetRotation=<rot 0 0.1712 0>, targetTranslation=<10.0146 0 0>}
       {targetRotation=<rot -85.1913 -138.5624 143.0618>, targetTranslation=<4.3575 -0.1466 -2.3223>}
       {targetRotation=<rot -64.7854 -3.089 6.0959>, targetTranslation=<15.5111 0 0>}
      ]
      customCurve {defaultValue=3.4028234663852886e+38, postInfinityExtrap=RCCE_Constant, preInfinityExtrap=RCCE_Constant, keys []}
      targetRotation=<rot 0 0 0>
     }
     {
      bApplyCustomCurve=false
      bIsHidden=false
      distanceMethod=DefaultMethod
      drivenName=chain_fwd
      functionType=DefaultFunction
      targetScale=0.42
      boneTransforms [
       {targetRotation=<rot 3.5363 -7.0955 -5.8687>, targetTranslation=<10.0146 0 0>}
       {targetRotation=<rot -69.6539 107.4577 -101.317>, targetTranslation=<4.3575 -0.1466 -2.3223>}
       {targetRotation=<rot -51.7801 55.3126 -43.8183>, targetTranslation=<15.5111 0 0>}
      ]
      customCurve {defaultValue=3.4028234663852886e+38, postInfinityExtrap=RCCE_Constant, preInfinityExtrap=RCCE_Constant, keys []}
      targetRotation=<rot 0 0 0>
     }
     {
      bApplyCustomCurve=false
      bIsHidden=false
      distanceMethod=DefaultMethod
      drivenName=chain_up
      functionType=DefaultFunction
      targetScale=1
      boneTransforms [
       {targetRotation=<rot -2.6635 1.8968 -6.5955>, targetTranslation=<10.0146 0 0>}
       {targetRotation=<rot -47.0161 -34.6019 22.8463>, targetTranslation=<4.3575 -0.1466 -2.3223>}
       {targetRotation=<rot -14.9181 89.5478 -68.8618>, targetTranslation=<15.5111 0 0>}
      ]
      customCurve {defaultValue=3.4028234663852886e+38, postInfinityExtrap=RCCE_Constant, preInfinityExtrap=RCCE_Constant, keys []}
      targetRotation=<rot 0 0 0>
     }
     {
      bApplyCustomCurve=false
      bIsHidden=false
      distanceMethod=DefaultMethod
      drivenName=chain_dwn
      functionType=DefaultFunction
      targetScale=0.42
      boneTransforms [
       {targetRotation=<rot -0.2733 -0.4939 -6.0688>, targetTranslation=<10.0146 0 0>}
       {targetRotation=<rot -74.5652 80.9031 -77.1513>, targetTranslation=<4.3575 -0.1466 -2.3223>}
       {targetRotation=<rot -48.4966 73.9275 -35.4824>, targetTranslation=<15.5111 0 0>}
      ]
      customCurve {defaultValue=3.4028234663852886e+38, postInfinityExtrap=RCCE_Constant, preInfinityExtrap=RCCE_Constant, keys []}
      targetRotation=<rot 0 0 0>
     }
    ]
    rBFParams {distanceMethod=Quaternion, function=Linear, radius=33.9354}
    sourceBones [{boneName=spine_05}, {boneName=clavicle_l}, {boneName=upperarm_l}]
   }
   n6 = Pose Driver - Source: spine_05(SourcePose=Use cached pose 'AnimCache'()) {
    poseAsset=&M_MED_DieselPunk_02_Arms_PoseAsset
    onlyDriveBones [{boneName=driver_pad_l}, {boneName=dyn_pad_bck}]
    poseTargets [
     {
      bApplyCustomCurve=false
      bIsHidden=false
      distanceMethod=DefaultMethod
      drivenName=bind
      functionType=DefaultFunction
      targetScale=0.7475
      boneTransforms [
       {targetRotation=<rot 0 0.1712 0>, targetTranslation=<10.0146 0 0>}
       {targetRotation=<rot -85.1913 -138.5624 143.0618>, targetTranslation=<4.3575 -0.1466 -2.3223>}
       {targetRotation=<rot -64.7854 -3.089 6.0959>, targetTranslation=<15.5111 0 0>}
      ]
      customCurve {defaultValue=3.4028234663852886e+38, postInfinityExtrap=RCCE_Constant, preInfinityExtrap=RCCE_Constant, keys []}
      targetRotation=<rot 0 0 0>
     }
     {
      bApplyCustomCurve=false
      bIsHidden=false
      distanceMethod=DefaultMethod
      drivenName=rl_fwd
      functionType=DefaultFunction
      targetScale=0.9072
      boneTransforms [
       {targetRotation=<rot -0.0865 -2.8461 -7.5396>, targetTranslation=<10.0146 0 0>}
       {targetRotation=<rot -68.0028 59.6671 -61.136>, targetTranslation=<4.3575 -0.1466 -2.3223>}
       {targetRotation=<rot -12.5757 76.5017 -72.6254>, targetTranslation=<15.5111 0 0>}
      ]
      customCurve {defaultValue=3.4028234663852886e+38, postInfinityExtrap=RCCE_Constant, preInfinityExtrap=RCCE_Constant, keys []}
      targetRotation=<rot 0 0 0>
     }
     {
      bApplyCustomCurve=false
      bIsHidden=false
      distanceMethod=DefaultMethod
      drivenName=rl_up
      functionType=DefaultFunction
      targetScale=1
      boneTransforms [
       {targetRotation=<rot 3.2144 -5.879 -7.3007>, targetTranslation=<10.0146 0 0>}
       {targetRotation=<rot -53.6035 43.5962 -49.4085>, targetTranslation=<4.3575 -0.1466 -2.3223>}
       {targetRotation=<rot 50.1808 34.9931 -71.021>, targetTranslation=<15.5111 0 0>}
      ]
      customCurve {defaultValue=3.4028234663852886e+38, postInfinityExtrap=RCCE_Constant, preInfinityExtrap=RCCE_Constant, keys []}
      targetRotation=<rot 0 0 0>
     }
     {
      bApplyCustomCurve=false
      bIsHidden=false
      distanceMethod=DefaultMethod
      drivenName=rl_dwn
      functionType=DefaultFunction
      targetScale=0.7475
      boneTransforms [
       {targetRotation=<rot 2.4032 -6.893 -6.4758>, targetTranslation=<10.0146 0 0>}
       {targetRotation=<rot -63.418 87.8771 -86.4949>, targetTranslation=<4.3575 -0.1466 -2.3223>}
       {targetRotation=<rot -70.8933 90.4854 -47.8981>, targetTranslation=<15.5111 0 0>}
      ]
      customCurve {defaultValue=3.4028234663852886e+38, postInfinityExtrap=RCCE_Constant, preInfinityExtrap=RCCE_Constant, keys []}
      targetRotation=<rot 0 0 0>
     }
    ]
    rBFParams {distanceMethod=Quaternion, function=Linear, radius=38.1528}
    sourceBones [{boneName=spine_05}, {boneName=clavicle_l}, {boneName=upperarm_l}]
   }
   Blend Poses by bool(
    BlendPose_0=@n6
    BlendPose_1=Blend Poses by bool(
     BlendPose_0=@n5
     BlendPose_1=Blend Poses by bool(BlendPose_0=@n3, BlendPose_1=@n2, bActiveValue=@n4) {blendType=Linear}
     bActiveValue=Get chaingun_active() -> (chaingun_active:bool)
    ) {blendType=Linear}
    bActiveValue=Get rl_active() -> (rl_active:bool)
   ) {blendType=Linear}
   "Transform (Modify) Bone - Bone: driver_pad_r"(Alpha=Get ArmPistolAlpha() -> (ArmPistolAlpha:float), Translation="3.852,-0.063,2.714", Rotation="-17.594,-0.245,3.797", Scale="1,1,1") {
    rotationMode=BMM_Additive
    rotationSpace=BCS_BoneSpace
    translationMode=BMM_Additive
    translationSpace=BCS_BoneSpace
    alphaBoolBlend {blendOption=Linear}
    boneToModify {boneName=driver_pad_r}
    rotation=<rot -17.594 -0.245 3.797>
    translation=<3.852 -0.063 2.714>
   }
   Pose Driver - Source: clavicle_l() {
    poseAsset=&M_MED_DieselPunk_02_Arms_PoseAsset
    onlyDriveBones [{boneName=driver_pad_l}, {boneName=dyn_pad_bck}]
    poseTargets [
     {
      bApplyCustomCurve=false
      bIsHidden=false
      distanceMethod=DefaultMethod
      drivenName=bind
      functionType=DefaultFunction
      targetScale=0.337
      boneTransforms [
       {targetRotation=<rot 0 0.1712 0>, targetTranslation=<10.0146 0 0>}
       {targetRotation=<rot -85.1913 -138.5624 143.0618>, targetTranslation=<4.3575 -0.1466 -2.3223>}
      ]
      customCurve {defaultValue=3.4028234663852886e+38, postInfinityExtrap=RCCE_Constant, preInfinityExtrap=RCCE_Constant, keys []}
      targetRotation=<rot 0 0 0>
     }
     {
      bApplyCustomCurve=false
      bIsHidden=false
      distanceMethod=DefaultMethod
      drivenName=pistol_fwd
      functionType=DefaultFunction
      targetScale=0.0642
      boneTransforms [
       {targetRotation=<rot -0.6935 -1.1975 -5.0607>, targetTranslation=<10.0146 0 0>}
       {targetRotation=<rot -80.1647 47.1898 -43.5433>, targetTranslation=<4.3575 -0.1466 -2.3223>}
      ]
      customCurve {defaultValue=3.4028234663852886e+38, postInfinityExtrap=RCCE_Constant, preInfinityExtrap=RCCE_Constant, keys []}
      targetRotation=<rot 0 0 0>
     }
     {
      bApplyCustomCurve=false
      bIsHidden=false
      distanceMethod=DefaultMethod
      drivenName=pistol_up
      functionType=DefaultFunction
      targetScale=1
      boneTransforms [
       {targetRotation=<rot -1.4676 -0.1641 -5.1562>, targetTranslation=<10.0146 0 0>}
       {targetRotation=<rot -45.5272 7.2881 -5.3683>, targetTranslation=<4.3575 -0.1466 -2.3223>}
      ]
      customCurve {defaultValue=3.4028234663852886e+38, postInfinityExtrap=RCCE_Constant, preInfinityExtrap=RCCE_Constant, keys []}
      targetRotation=<rot 0 0 0>
     }
     {
      bApplyCustomCurve=false
      bIsHidden=false
      distanceMethod=DefaultMethod
      drivenName=pistol_dwn
      functionType=DefaultFunction
      targetScale=0.0642
      boneTransforms [
       {targetRotation=<rot -0.9888 -1.1873 -5.9157>, targetTranslation=<10.0146 0 0>}
       {targetRotation=<rot -81.543 57.4431 -54.3032>, targetTranslation=<4.3575 -0.1466 -2.3223>}
      ]
      customCurve {defaultValue=3.4028234663852886e+38, postInfinityExtrap=RCCE_Constant, preInfinityExtrap=RCCE_Constant, keys []}
      targetRotation=<rot 0 0 0>
     }
     {
      bApplyCustomCurve=false
      bIsHidden=false
      distanceMethod=DefaultMethod
      drivenName=pistol_bck
      functionType=DefaultFunction
      targetScale=0.1575
      boneTransforms [
       {targetRotation=<rot -0.196 -2.3231 -1.768>, targetTranslation=<10.0146 0 0>}
       {targetRotation=<rot -82.1915 81.681 -80.2223>, targetTranslation=<4.3575 -0.1466 -2.3223>}
      ]
      customCurve {defaultValue=3.4028234663852886e+38, postInfinityExtrap=RCCE_Constant, preInfinityExtrap=RCCE_Constant, keys []}
      targetRotation=<rot 0 0 0>
     }
    ]
    rBFParams {distanceMethod=Quaternion, radius=18.6924}
    sourceBones [{boneName=clavicle_l}, {boneName=upperarm_l}]
   }
   Pose Driver - Source: clavicle_l() {
    poseAsset=&M_MED_DieselPunk_02_Arms_PoseAsset
    onlyDriveBones [{boneName=driver_pad_l}, {boneName=dyn_pad_bck}]
    poseTargets [
     {
      bApplyCustomCurve=false
      bIsHidden=false
      distanceMethod=DefaultMethod
      drivenName=bind
      functionType=DefaultFunction
      targetScale=0.6488
      boneTransforms [
       {targetRotation=<rot -85.1913 -138.5624 143.0618>, targetTranslation=<4.3575 -0.1466 -2.3223>}
       {targetRotation=<rot -64.7854 -3.089 6.0959>, targetTranslation=<15.5111 0 0>}
      ]
      customCurve {defaultValue=3.4028234663852886e+38, postInfinityExtrap=RCCE_Constant, preInfinityExtrap=RCCE_Constant, keys []}
      targetRotation=<rot 0 0 0>
     }
     {
      bApplyCustomCurve=false
      bIsHidden=false
      distanceMethod=DefaultMethod
      drivenName=pistol_fwd
      functionType=DefaultFunction
      targetScale=0.5513
      boneTransforms [
       {targetRotation=<rot -80.1647 47.1898 -43.5433>, targetTranslation=<4.3575 -0.1466 -2.3223>}
       {targetRotation=<rot -27.4019 74.8174 -63.7631>, targetTranslation=<15.5111 0 0>}
      ]
      customCurve {defaultValue=3.4028234663852886e+38, postInfinityExtrap=RCCE_Constant, preInfinityExtrap=RCCE_Constant, keys []}
      targetRotation=<rot 0 0 0>
     }
     {
      bApplyCustomCurve=false
      bIsHidden=false
      distanceMethod=DefaultMethod
      drivenName=pistol_up
      functionType=DefaultFunction
      targetScale=1
      boneTransforms [
       {targetRotation=<rot -45.5272 7.2881 -5.3683>, targetTranslation=<4.3575 -0.1466 -2.3223>}
       {targetRotation=<rot 9.8617 54.2947 -96.1665>, targetTranslation=<15.5111 0 0>}
      ]
      customCurve {defaultValue=3.4028234663852886e+38, postInfinityExtrap=RCCE_Constant, preInfinityExtrap=RCCE_Constant, keys []}
      targetRotation=<rot 0 0 0>
     }
     {
      bApplyCustomCurve=false
      bIsHidden=false
      distanceMethod=DefaultMethod
      drivenName=pistol_dwn
      functionType=DefaultFunction
      targetScale=0.5702
      boneTransforms [
       {targetRotation=<rot -81.543 57.4431 -54.3032>, targetTranslation=<4.3575 -0.1466 -2.3223>}
       {targetRotation=<rot -70.1236 89.2767 -56.32>, targetTranslation=<15.5111 0 0>}
      ]
      customCurve {defaultValue=3.4028234663852886e+38, postInfinityExtrap=RCCE_Constant, preInfinityExtrap=RCCE_Constant, keys []}
      targetRotation=<rot 0 0 0>
     }
     {
      bApplyCustomCurve=false
      bIsHidden=false
      distanceMethod=DefaultMethod
      drivenName=pistol_bck
      functionType=DefaultFunction
      targetScale=0.5513
      boneTransforms [
       {targetRotation=<rot -82.1915 81.681 -80.2223>, targetTranslation=<4.3575 -0.1466 -2.3223>}
       {targetRotation=<rot -36.6606 93.9397 -43.0028>, targetTranslation=<15.5111 0 0>}
      ]
      customCurve {defaultValue=3.4028234663852886e+38, postInfinityExtrap=RCCE_Constant, preInfinityExtrap=RCCE_Constant, keys []}
      targetRotation=<rot 0 0 0>
     }
    ]
    rBFParams {distanceMethod=Quaternion, radius=33.1614}
    sourceBones [{boneName=clavicle_l}, {boneName=upperarm_l}]
   }
   Pose Driver - Source: clavicle_r!(SourcePose:Pose Link Structure) -> (Pose:Pose Link Structure) {
    is=title AnimGraphNode_PoseDriver
    poseAsset=&/M_MED_DieselPunk_02_PadR_PoseAsset
    onlyDriveBones [{boneName=dyn_pad_r}]
    poseTargets [
     {
      bApplyCustomCurve=false
      bIsHidden=false
      distanceMethod=DefaultMethod
      drivenName=bind
      functionType=DefaultFunction
      targetScale=0.3439
      boneTransforms [
       {targetRotation=<rot -85.1914 41.4375 143.0618>, targetTranslation=<4.3575 -0.1466 2.3223>}
       {targetRotation=<rot -64.7854 -3.089 6.0959>, targetTranslation=<-15.5111 0 0>}
      ]
      customCurve {defaultValue=3.4028234663852886e+38, postInfinityExtrap=RCCE_Constant, preInfinityExtrap=RCCE_Constant, keys []}
      targetRotation=<rot 0 0 0>
     }
     {
      bApplyCustomCurve=false
      bIsHidden=false
      distanceMethod=DefaultMethod
      drivenName=Pose_0
      functionType=DefaultFunction
      targetScale=0.6296
      boneTransforms [
       {targetRotation=<rot -64.576 87.1289 99.0964>, targetTranslation=<4.3575 -0.1466 2.3223>}
       {targetRotation=<rot -3.4413 0.4629 -58.831>, targetTranslation=<-15.5111 0 0>}
      ]
      customCurve {defaultValue=3.4028234663852886e+38, postInfinityExtrap=RCCE_Constant, preInfinityExtrap=RCCE_Constant, keys []}
      targetRotation=<rot 0 0 0>
     }
     {
      bApplyCustomCurve=false
      bIsHidden=false
      distanceMethod=DefaultMethod
      drivenName=Pose_1
      functionType=DefaultFunction
      targetScale=1
      boneTransforms [
       {targetRotation=<rot -67.6431 100.198 84.7865>, targetTranslation=<4.3575 -0.1466 2.3223>}
       {targetRotation=<rot 12.0773 82.7237 -53.9719>, targetTranslation=<-15.5111 0 0>}
      ]
      customCurve {defaultValue=3.4028234663852886e+38, postInfinityExtrap=RCCE_Constant, preInfinityExtrap=RCCE_Constant, keys []}
      targetRotation=<rot 0 0 0>
     }
     {
      bApplyCustomCurve=false
      bIsHidden=false
      distanceMethod=DefaultMethod
      drivenName=Pose_2
      functionType=DefaultFunction
      targetScale=0.6105
      boneTransforms [
       {targetRotation=<rot -63.3746 84.7974 101.3322>, targetTranslation=<4.3575 -0.1466 2.3223>}
       {targetRotation=<rot -21.9366 -23.6672 15.5356>, targetTranslation=<-15.5111 0 0>}
      ]
      customCurve {defaultValue=3.4028234663852886e+38, postInfinityExtrap=RCCE_Constant, preInfinityExtrap=RCCE_Constant, keys []}
      targetRotation=<rot 0 0 0>
     }
     {
      bApplyCustomCurve=false
      bIsHidden=false
      distanceMethod=DefaultMethod
      drivenName=Pose_3
      functionType=DefaultFunction
      targetScale=0.3439
      boneTransforms [
       {targetRotation=<rot -85.1916 41.4356 143.0641>, targetTranslation=<4.3575 -0.1466 2.3223>}
       {targetRotation=<rot -53.4049 -50.3444 59.6811>, targetTranslation=<-15.5111 0 0>}
      ]
      customCurve {defaultValue=3.4028234663852886e+38, postInfinityExtrap=RCCE_Constant, preInfinityExtrap=RCCE_Constant, keys []}
      targetRotation=<rot 0 0 0>
     }
     {
      bApplyCustomCurve=false
      bIsHidden=false
      distanceMethod=DefaultMethod
      drivenName=Pose_4
      functionType=DefaultFunction
      targetScale=0.6296
      boneTransforms [
       {targetRotation=<rot -64.4955 -165.3014 -9.4212>, targetTranslation=<4.3575 -0.1466 2.3223>}
       {targetRotation=<rot 22.1549 -6.934 -79.336>, targetTranslation=<-15.5111 0 0>}
      ]
      customCurve {defaultValue=3.4028234663852886e+38, postInfinityExtrap=RCCE_Constant, preInfinityExtrap=RCCE_Constant, keys []}
      targetRotation=<rot 0 0 0>
     }
    ]
    rBFParams {distanceMethod=Quaternion, radius=41.7326}
    sourceBones [{boneName=clavicle_r}, {boneName=upperarm_r}]
   }
   Pose Driver - Source: spine_05() {
    poseAsset=&M_MED_DieselPunk_02_Arms_PoseAsset
    onlyDriveBones [{boneName=driver_pad_l}, {boneName=dyn_pad_bck}]
    poseTargets [
     {
      bApplyCustomCurve=false
      bIsHidden=false
      distanceMethod=DefaultMethod
      drivenName=bind
      functionType=DefaultFunction
      targetScale=0.5292
      boneTransforms [
       {targetRotation=<rot 0 0.1712 0>, targetTranslation=<10.0146 0 0>}
       {targetRotation=<rot -85.1913 -138.5624 143.0618>, targetTranslation=<4.3575 -0.1466 -2.3223>}
       {targetRotation=<rot -64.7854 -3.089 6.0959>, targetTranslation=<15.5111 0 0>}
      ]
      customCurve {defaultValue=3.4028234663852886e+38, postInfinityExtrap=RCCE_Constant, preInfinityExtrap=RCCE_Constant, keys []}
      targetRotation=<rot 0 0 0>
     }
     {
      bApplyCustomCurve=false
      bIsHidden=false
      distanceMethod=DefaultMethod
      drivenName=rifle_fwd
      functionType=DefaultFunction
      targetScale=0.5361
      boneTransforms [
       {targetRotation=<rot -1.8763 -1.5677 -6.688>, targetTranslation=<10.0146 0 0>}
       {targetRotation=<rot -82.3641 52.539 -51.6926>, targetTranslation=<4.3575 -0.1466 -2.3223>}
       {targetRotation=<rot -37.0534 63.6677 -82.6126>, targetTranslation=<15.5111 0 0>}
      ]
      customCurve {defaultValue=3.4028234663852886e+38, postInfinityExtrap=RCCE_Constant, preInfinityExtrap=RCCE_Constant, keys []}
      targetRotation=<rot 0 0 0>
     }
     {
      bApplyCustomCurve=false
      bIsHidden=false
      distanceMethod=DefaultMethod
      drivenName=rifle_up
      functionType=DefaultFunction
      targetScale=0.4321
      boneTransforms [
       {targetRotation=<rot -2.8104 -0.2929 -6.8311>, targetTranslation=<10.0146 0 0>}
       {targetRotation=<rot -51.6841 10.4523 -9.0777>, targetTranslation=<4.3575 -0.1466 -2.3223>}
       {targetRotation=<rot -7.9729 51.1321 -83.85>, targetTranslation=<15.5111 0 0>}
      ]
      customCurve {defaultValue=3.4028234663852886e+38, postInfinityExtrap=RCCE_Constant, preInfinityExtrap=RCCE_Constant, keys []}
      targetRotation=<rot 0 0 0>
     }
     {
      bApplyCustomCurve=false
      bIsHidden=false
      distanceMethod=DefaultMethod
      drivenName=rifle_dwn
      functionType=DefaultFunction
      targetScale=0.3834
      boneTransforms [
       {targetRotation=<rot -2.6929 -0.7332 -8.0647>, targetTranslation=<10.0146 0 0>}
       {targetRotation=<rot -82.4636 143.5654 -137.6768>, targetTranslation=<4.3575 -0.1466 -2.3223>}
       {targetRotation=<rot -67.2141 49.9006 -28.3212>, targetTranslation=<15.5111 0 0>}
      ]
      customCurve {defaultValue=3.4028234663852886e+38, postInfinityExtrap=RCCE_Constant, preInfinityExtrap=RCCE_Constant, keys []}
      targetRotation=<rot 0 0 0>
     }
     {
      bApplyCustomCurve=false
      bIsHidden=false
      distanceMethod=DefaultMethod
      drivenName=pistol_fwd
      functionType=DefaultFunction
      targetScale=0.4153
      boneTransforms [
       {targetRotation=<rot -0.6935 -1.1975 -5.0607>, targetTranslation=<10.0146 0 0>}
       {targetRotation=<rot -80.1647 47.1898 -43.5433>, targetTranslation=<4.3575 -0.1466 -2.3223>}
       {targetRotation=<rot -27.4019 74.8174 -63.7631>, targetTranslation=<15.5111 0 0>}
      ]
      customCurve {defaultValue=3.4028234663852886e+38, postInfinityExtrap=RCCE_Constant, preInfinityExtrap=RCCE_Constant, keys []}
      targetRotation=<rot 0 0 0>
     }
     {
      bApplyCustomCurve=false
      bIsHidden=false
      distanceMethod=DefaultMethod
      drivenName=pistol_up
      functionType=DefaultFunction
      targetScale=0.4321
      boneTransforms [
       {targetRotation=<rot -1.4676 -0.1641 -5.1562>, targetTranslation=<10.0146 0 0>}
       {targetRotation=<rot -45.5272 7.2881 -5.3683>, targetTranslation=<4.3575 -0.1466 -2.3223>}
       {targetRotation=<rot 9.8617 54.2947 -96.1665>, targetTranslation=<15.5111 0 0>}
      ]
      customCurve {defaultValue=3.4028234663852886e+38, postInfinityExtrap=RCCE_Constant, preInfinityExtrap=RCCE_Constant, keys []}
      targetRotation=<rot 0 0 0>
     }
     {
      bApplyCustomCurve=false
      bIsHidden=false
      distanceMethod=DefaultMethod
      drivenName=pistol_dwn
      functionType=DefaultFunction
      targetScale=0.3834
      boneTransforms [
       {targetRotation=<rot -0.9888 -1.1873 -5.9157>, targetTranslation=<10.0146 0 0>}
       {targetRotation=<rot -81.543 57.4431 -54.3032>, targetTranslation=<4.3575 -0.1466 -2.3223>}
       {targetRotation=<rot -70.1236 89.2767 -56.32>, targetTranslation=<15.5111 0 0>}
      ]
      customCurve {defaultValue=3.4028234663852886e+38, postInfinityExtrap=RCCE_Constant, preInfinityExtrap=RCCE_Constant, keys []}
      targetRotation=<rot 0 0 0>
     }
     {
      bApplyCustomCurve=false
      bIsHidden=false
      distanceMethod=DefaultMethod
      drivenName=pistol_bck
      functionType=DefaultFunction
      targetScale=0.4103
      boneTransforms [
       {targetRotation=<rot -0.196 -2.3231 -1.768>, targetTranslation=<10.0146 0 0>}
       {targetRotation=<rot -82.1915 81.681 -80.2223>, targetTranslation=<4.3575 -0.1466 -2.3223>}
       {targetRotation=<rot -36.6606 93.9397 -43.0028>, targetTranslation=<15.5111 0 0>}
      ]
      customCurve {defaultValue=3.4028234663852886e+38, postInfinityExtrap=RCCE_Constant, preInfinityExtrap=RCCE_Constant, keys []}
      targetRotation=<rot 0 0 0>
     }
     {
      bApplyCustomCurve=false
      bIsHidden=false
      distanceMethod=DefaultMethod
      drivenName=rl_fwd
      functionType=DefaultFunction
      targetScale=0.4153
      boneTransforms [
       {targetRotation=<rot -0.0865 -2.8461 -7.5396>, targetTranslation=<10.0146 0 0>}
       {targetRotation=<rot -68.0028 59.6671 -61.136>, targetTranslation=<4.3575 -0.1466 -2.3223>}
       {targetRotation=<rot -12.5757 76.5017 -72.6254>, targetTranslation=<15.5111 0 0>}
      ]
      customCurve {defaultValue=3.4028234663852886e+38, postInfinityExtrap=RCCE_Constant, preInfinityExtrap=RCCE_Constant, keys []}
      targetRotation=<rot 0 0 0>
     }
     {
      bApplyCustomCurve=false
      bIsHidden=false
      distanceMethod=DefaultMethod
      drivenName=rl_up
      functionType=DefaultFunction
      targetScale=1
      boneTransforms [
       {targetRotation=<rot 3.2144 -5.879 -7.3007>, targetTranslation=<10.0146 0 0>}
       {targetRotation=<rot -53.6035 43.5962 -49.4085>, targetTranslation=<4.3575 -0.1466 -2.3223>}
       {targetRotation=<rot 50.1808 34.9931 -71.021>, targetTranslation=<15.5111 0 0>}
      ]
      customCurve {defaultValue=3.4028234663852886e+38, postInfinityExtrap=RCCE_Constant, preInfinityExtrap=RCCE_Constant, keys []}
      targetRotation=<rot 0 0 0>
     }
     {
      bApplyCustomCurve=false
      bIsHidden=false
      distanceMethod=DefaultMethod
      drivenName=rl_dwn
      functionType=DefaultFunction
      targetScale=0.4341
      boneTransforms [
       {targetRotation=<rot 2.4032 -6.893 -6.4758>, targetTranslation=<10.0146 0 0>}
       {targetRotation=<rot -63.418 87.8771 -86.4949>, targetTranslation=<4.3575 -0.1466 -2.3223>}
       {targetRotation=<rot -70.8933 90.4854 -47.8981>, targetTranslation=<15.5111 0 0>}
      ]
      customCurve {defaultValue=3.4028234663852886e+38, postInfinityExtrap=RCCE_Constant, preInfinityExtrap=RCCE_Constant, keys []}
      targetRotation=<rot 0 0 0>
     }
     {
      bApplyCustomCurve=false
      bIsHidden=false
      distanceMethod=DefaultMethod
      drivenName=chain_fwd
      functionType=DefaultFunction
      targetScale=0.4861
      boneTransforms [
       {targetRotation=<rot 3.5363 -7.0955 -5.8687>, targetTranslation=<10.0146 0 0>}
       {targetRotation=<rot -69.6539 107.4577 -101.317>, targetTranslation=<4.3575 -0.1466 -2.3223>}
       {targetRotation=<rot -51.7801 55.3126 -43.8183>, targetTranslation=<15.5111 0 0>}
      ]
      customCurve {defaultValue=3.4028234663852886e+38, postInfinityExtrap=RCCE_Constant, preInfinityExtrap=RCCE_Constant, keys []}
      targetRotation=<rot 0 0 0>
     }
     {
      bApplyCustomCurve=false
      bIsHidden=false
      distanceMethod=DefaultMethod
      drivenName=chain_up
      functionType=DefaultFunction
      targetScale=0.9
      boneTransforms [
       {targetRotation=<rot -2.6635 1.8968 -6.5955>, targetTranslation=<10.0146 0 0>}
       {targetRotation=<rot -47.0161 -34.6019 22.8463>, targetTranslation=<4.3575 -0.1466 -2.3223>}
       {targetRotation=<rot -14.9181 89.5478 -68.8618>, targetTranslation=<15.5111 0 0>}
      ]
      customCurve {defaultValue=3.4028234663852886e+38, postInfinityExtrap=RCCE_Constant, preInfinityExtrap=RCCE_Constant, keys []}
      targetRotation=<rot 0 0 0>
     }
     {
      bApplyCustomCurve=false
      bIsHidden=false
      distanceMethod=DefaultMethod
      drivenName=chain_dwn
      functionType=DefaultFunction
      targetScale=0.4103
      boneTransforms [
       {targetRotation=<rot -0.2733 -0.4939 -6.0688>, targetTranslation=<10.0146 0 0>}
       {targetRotation=<rot -74.5652 80.9031 -77.1513>, targetTranslation=<4.3575 -0.1466 -2.3223>}
       {targetRotation=<rot -48.4966 73.9275 -35.4824>, targetTranslation=<15.5111 0 0>}
      ]
      customCurve {defaultValue=3.4028234663852886e+38, postInfinityExtrap=RCCE_Constant, preInfinityExtrap=RCCE_Constant, keys []}
      targetRotation=<rot 0 0 0>
     }
     {
      bApplyCustomCurve=false
      bIsHidden=false
      distanceMethod=DefaultMethod
      drivenName=sky_fall
      functionType=DefaultFunction
      targetScale=0.9638
      boneTransforms [
       {targetRotation=<rot 0.2027 0.1178 0.0426>, targetTranslation=<10.0146 0 0>}
       {targetRotation=<rot -63.1095 -93.8352 100.0777>, targetTranslation=<4.3575 -0.1466 -2.3223>}
       {targetRotation=<rot -1.3562 3.6287 -80.1244>, targetTranslation=<15.5111 0 0>}
      ]
      customCurve {defaultValue=3.4028234663852886e+38, postInfinityExtrap=RCCE_Constant, preInfinityExtrap=RCCE_Constant, keys []}
      targetRotation=<rot 0 0 0>
     }
     {
      bApplyCustomCurve=false
      bIsHidden=false
      distanceMethod=DefaultMethod
      drivenName=sky_fwd
      functionType=DefaultFunction
      targetScale=0.9189
      boneTransforms [
       {targetRotation=<rot -0.4478 -8.7405 0.0659>, targetTranslation=<10.0146 0 0>}
       {targetRotation=<rot -64.6511 -93.4838 99.6726>, targetTranslation=<4.3575 -0.1466 -2.3223>}
       {targetRotation=<rot -25.114 -39.8062 -5.9033>, targetTranslation=<15.5111 0 0>}
      ]
      customCurve {defaultValue=3.4028234663852886e+38, postInfinityExtrap=RCCE_Constant, preInfinityExtrap=RCCE_Constant, keys []}
      targetRotation=<rot 0 0 0>
     }
     {
      bApplyCustomCurve=false
      bIsHidden=false
      distanceMethod=DefaultMethod
      drivenName=sky_dive
      functionType=DefaultFunction
      targetScale=0.634
      boneTransforms [
       {targetRotation=<rot 0.6195 3.5554 0.2064>, targetTranslation=<10.0146 0 0>}
       {targetRotation=<rot -85.1958 -138.5355 143.0312>, targetTranslation=<4.3575 -0.1466 -2.3223>}
       {targetRotation=<rot -43.0187 -47.2076 35.696>, targetTranslation=<15.5111 0 0>}
      ]
      customCurve {defaultValue=3.4028234663852886e+38, postInfinityExtrap=RCCE_Constant, preInfinityExtrap=RCCE_Constant, keys []}
      targetRotation=<rot 0 0 0>
     }
     {
      bApplyCustomCurve=false
      bIsHidden=false
      distanceMethod=DefaultMethod
      drivenName=sky_glide
      functionType=DefaultFunction
      targetScale=0.9638
      boneTransforms [
       {targetRotation=<rot 0.9801 -1.0847 -0.1213>, targetTranslation=<10.0146 0 0>}
       {targetRotation=<rot -63.3118 15.0485 -9.8732>, targetTranslation=<4.3575 -0.1466 -2.3223>}
       {targetRotation=<rot 22.8743 -7.2564 -77.158>, targetTranslation=<15.5111 0 0>}
      ]
      customCurve {defaultValue=3.4028234663852886e+38, postInfinityExtrap=RCCE_Constant, preInfinityExtrap=RCCE_Constant, keys []}
      targetRotation=<rot 0 0 0>
     }
    ]
    rBFParams {distanceMethod=Quaternion, function=Linear, radius=26.5974}
    sourceBones [{boneName=spine_05}, {boneName=clavicle_l}, {boneName=upperarm_l}]
   }
   n7 = Get ArmPistolChestAlpha() -> (ArmPistolChestAlpha:float)
   n8 = Pose Driver - Source: upperarm_r!(SourcePose:Pose Link Structure=Use cached pose 'AnimCache'()) -> (Pose:Pose Link Structure) {
    is=title AnimGraphNode_PoseDriver
    poseAsset=&M_MED_DieselPunk_02_PadR_PoseAsset
    onlyDriveBones [{boneName=dyn_pad_r}]
    poseTargets [
     {
      bApplyCustomCurve=false
      bIsHidden=false
      distanceMethod=DefaultMethod
      drivenName=bind
      functionType=DefaultFunction
      targetScale=1
      boneTransforms [{targetRotation=<rot -64.7854 -3.089 6.0959>, targetTranslation=<-15.5111 0 0>}]
      customCurve {defaultValue=3.4028234663852886e+38, postInfinityExtrap=RCCE_Constant, preInfinityExtrap=RCCE_Constant, keys []}
      targetRotation=<rot 0 0 0>
     }
     {
      bApplyCustomCurve=false
      bIsHidden=false
      distanceMethod=DefaultMethod
      drivenName=Pose_0
      functionType=DefaultFunction
      targetScale=0.3832
      boneTransforms [{targetRotation=<rot -3.4413 0.4629 -58.831>, targetTranslation=<-15.5111 0 0>}]
      customCurve {defaultValue=3.4028234663852886e+38, postInfinityExtrap=RCCE_Constant, preInfinityExtrap=RCCE_Constant, keys []}
      targetRotation=<rot 0 0 0>
     }
     {
      bApplyCustomCurve=false
      bIsHidden=false
      distanceMethod=DefaultMethod
      drivenName=Pose_7
      functionType=DefaultFunction
      targetScale=0.3832
      boneTransforms [{targetRotation=<rot 22.1549 -6.934 -79.336>, targetTranslation=<-15.5111 0 0>}]
      customCurve {defaultValue=3.4028234663852886e+38, postInfinityExtrap=RCCE_Constant, preInfinityExtrap=RCCE_Constant, keys []}
      targetRotation=<rot 0 0 0>
     }
    ]
    rBFParams {distanceMethod=Quaternion, radius=100}
    sourceBones [{boneName=upperarm_r}]
   }
   n9 = Use cached pose 'AnimCache'()
   n10 = Get bIsSkyDiving()
   n11 = "Constraint - driver_dyn_zipper_r (PPPPPP)"!(
    ComponentPose=Local To Component(LocalPose=Blend Poses by bool(BlendPose_0=@n8, BlendPose_1=@n9, bActiveValue=@n10) {blendType=Linear})
    Alpha=1
    ConstraintWeights_0=0.5444
    ConstraintWeights_1:float=0.0116
    ConstraintWeights_2:float=0.0661
    ConstraintWeights_3:float=0.0968
    ConstraintWeights_4:float=0.0092
    ConstraintWeights_5:float=0.2719
   ) -> (Pose) {
    alphaBoolBlend {blendOption=Linear}
    boneToModify {boneName=driver_dyn_zipper_r}
    constraintSetup [
     {targetBone {boneName=clavicle_fix_r}}
     {targetBone {boneName=spine_02}}
     {targetBone {boneName=spine_03}}
     {targetBone {boneName=spine_04}}
     {targetBone {boneName=spine_05}}
     {targetBone {boneName=clavicle_r}}
    ]
    constraintWeights [1,1,1,1,1,1]
   }
   n12 = "Constraint - driver_pad_r (PPPP)"!(
    ComponentPose=@n11
    Alpha=1
    ConstraintWeights_0=0.8284
    ConstraintWeights_1:float=0.0968
    ConstraintWeights_2:float=0.0469
    ConstraintWeights_3:float=0.0279
   ) -> (Pose) {
    alphaBoolBlend {blendOption=Linear}
    boneToModify {boneName=driver_pad_r}
    constraintSetup [
     {targetBone {boneName=clavicle_r}}
     {targetBone {boneName=upperarm_twist_01_r}}
     {targetBone {boneName=clavicle_fix_r}}
     {targetBone {boneName=spine_05}}
    ]
    constraintWeights [1,1,1,1]
   }
   n13 = "Constraint - dyn_flapB_l (PPPP)"!(
    ComponentPose=@n12
    Alpha=1
    ConstraintWeights_0=0.0865
    ConstraintWeights_1:float=0.3173
    ConstraintWeights_2:float=0.184
    ConstraintWeights_3:float=0.4122
   ) -> (Pose) {
    alphaBoolBlend {blendOption=Linear}
    boneToModify {boneName=dyn_flapB_l}
    constraintSetup [{targetBone {boneName=spine_03}}, {targetBone {boneName=spine_04}}, {targetBone {boneName=spine_05}}, {targetBone {boneName=clavicle_l}}]
    constraintWeights [1,1,1,1]
   }
   n14 = "Constraint - dyn_flapA_l (PPPPP)"!(
    ComponentPose=@n13
    Alpha=1
    ConstraintWeights_0=0.0093
    ConstraintWeights_1:float=0.0333
    ConstraintWeights_2:float=0.3158
    ConstraintWeights_3:float=0.021
    ConstraintWeights_4:float=0.6206
   ) -> (Pose) {
    alphaBoolBlend {blendOption=Linear}
    boneToModify {boneName=dyn_flapA_l}
    constraintSetup [
     {targetBone {boneName=spine_02}}
     {targetBone {boneName=spine_03}}
     {targetBone {boneName=spine_04}}
     {targetBone {boneName=spine_05}}
     {targetBone {boneName=clavicle_l}}
    ]
    constraintWeights [1,1,1,1,1]
   }
   n15 = Anim Dynamics - Bone: dyn_hip_pack_r(ComponentPose=@n14, Alpha=Get SpeedMaskAlpha() -> (SpeedMaskAlpha:float), ExternalForce="0,0,0") {
    angularBiasOverride=0.75
    angularDampingOverride=0.9
    angularSpringConstant=250
    bAngularSpring=true
    bGravityOverrideInSimSpace=true
    bOverrideAngularBias=true
    bOverrideAngularDamping=true
    gravityScale=0.4
    simulationSpace=RootRelative
    alphaBoolBlend {blendOption=Linear}
    boundBone {boneName=dyn_hip_pack_r}
    componentAppliedLinearAccClamp=<1000 1000 1000>
    componentLinearVelScale=<50 50 50>
    physicsBodyDefinitions [{
      sphereCollisionRadius=0
      boundBone {boneName=dyn_hip_pack_r}
      boxExtents=<50 50 50>
      constraintSetup {angularLimitsMax=<10 20 8>, angularLimitsMin=<-10 0 -8>, angularTarget=<1 0 0>}
      localJointOffset=<-0.2 -0.15 -25>
     }]
   }
   n16 = RigidBody(ComponentPose=@n15) {
    bOverrideWorldGravity=true
    overridePhysicsAsset=&/M_MED_Dieselpunk_02_Physics
    simulationSpace=BaseBoneSpace
    alphaBoolBlend {blendOption=Linear}
    baseBoneRef {boneName=spine_01}
    componentAppliedLinearAccClamp=<1000 1000 1000>
    componentLinearAccScale=<10 10 10>
    componentLinearVelScale=<10 10 10>
    overrideWorldGravity {z=-980}
   }
   n17 = "Constraint - dyn_mask_01 (PP)"!(ComponentPose=@n16, Alpha=1, ConstraintWeights_0=0.1028, ConstraintWeights_1:float=0.8972) -> (Pose) {
    alphaBoolBlend {blendOption=Linear}
    boneToModify {boneName=dyn_mask_01}
    constraintSetup [{targetBone {boneName=pelvis}}, {targetBone {boneName=thigh_twist_01_l}}]
    constraintWeights [1,1]
   }
   n18 = "Transform (Modify) Bone - Bone: driver_dyn_mask_01"(
    ComponentPose=@n17
    Alpha=Get CrouchAlpha() -> (CrouchAlpha:float)
    Translation="-1.542,2.433,-1.947"
    Rotation="-23.301,-25.523,32.798"
    Scale="1,1,1"
   ) {
    rotationMode=BMM_Additive
    translationMode=BMM_Additive
    alphaBoolBlend {blendOption=Linear}
    alphaScaleBiasClamp {bInterpResult=true}
    boneToModify {boneName=driver_dyn_mask_01}
    rotation=<rot -23.301 -25.5226 32.7983>
    translation=<-1.5419 2.433 -1.947>
   }
   n19 = Anim Dynamics - Bone: dyn_mask_01!(
    ComponentPose=@n18
    Alpha=Get SpeedMaskAlpha()
    AngularSpringConstant:float=Get SpeedMaskSpring() -> (SpeedMaskSpring:float)
    ExternalForce="0,0,0"
   ) -> (Pose) {
    angularBiasOverride=0.7
    angularDampingOverride=0.9
    angularSpringConstant=300
    bAngularSpring=true
    bGravityOverrideInSimSpace=true
    bOverrideAngularBias=true
    bOverrideAngularDamping=true
    bUseGravityOverride=true
    gravityScale=0.33
    simulationSpace=BoneRelative
    alphaBoolBlend {blendOption=Linear}
    alphaScaleBiasClamp {bInterpResult=true, interpSpeedDecreasing=8, interpSpeedIncreasing=8}
    boundBone {boneName=dyn_mask_01}
    componentAppliedLinearAccClamp=<10 10 10>
    componentLinearAccScale=<1 1 0.1>
    componentLinearVelScale=<1 1 1>
    gravityOverride {z=-980}
    physicsBodyDefinitions [{
      sphereCollisionRadius=0
      boundBone {boneName=dyn_mask_01}
      boxExtents=<40 80 40>
      constraintSetup {angularLimitsMax=<3 0 45>, angularLimitsMin=<-3 -89 -45>, angularTarget=<1 0 0>}
      localJointOffset=<20 -0.1 0>
     }]
    relativeSpaceBone {boneName=root}
   }
   n20 = "Transform (Modify) Bone - Bone: clavicle_fix_r"(ComponentPose=@n19, Alpha=@n7, Translation="2.847,-2.044,0.623", Rotation="-5.750,2.638,-8.972", Scale="1,1,1") {
    rotationMode=BMM_Additive
    translationMode=BMM_Additive
    alphaBoolBlend {blendOption=Linear}
    alphaScaleBiasClamp {bInterpResult=true, interpSpeedDecreasing=20, interpSpeedIncreasing=20}
    boneToModify {boneName=clavicle_fix_r}
    rotation=<rot -5.7502 2.638 -8.9722>
    translation=<2.847 -2.044 0.623>
   }
   n21 = "Transform (Modify) Bone - Bone: dyn_clav_strap_r"(ComponentPose=@n20, Alpha=@n7, Translation="0,0,0", Rotation="5.775,17.272,20.675", Scale="1,1,1") {
    rotationMode=BMM_Additive
    rotationSpace=BCS_BoneSpace
    translationMode=BMM_Additive
    alphaBoolBlend {blendOption=Linear}
    alphaScaleBiasClamp {bInterpResult=true, interpSpeedDecreasing=20, interpSpeedIncreasing=20}
    boneToModify {boneName=dyn_clav_strap_r}
    rotation=<rot 5.775 17.2722 20.6749>
   }
   n22 = "Transform (Modify) Bone - Bone: driver_pad_r"(ComponentPose=@n21, Alpha=1, Translation="3.852,-0.063,2.714", Rotation="-16.444,-7.353,2.745", Scale="1,1,1") {
    rotationMode=BMM_Additive
    rotationSpace=BCS_BoneSpace
    translationMode=BMM_Additive
    translationSpace=BCS_BoneSpace
    alphaBoolBlend {blendOption=Linear}
    boneToModify {boneName=driver_pad_r}
    rotation=<rot -16.444 -7.353 2.745>
    translation=<3.852 -0.063 2.714>
   }
   n23 = "Transform (Modify) Bone - Bone: dyn_pad_frt_r"(
    ComponentPose=@n22
    Alpha=Get ArmPistolPadAlpha() -> (ArmPistolPadAlpha:float)
    Translation="0,0,0"
    Rotation="6.427,14.408,-45"
    Scale="1,1,1"
   ) {
    rotationMode=BMM_Additive
    alphaBoolBlend {blendOption=Linear}
    alphaScaleBiasClamp {bInterpResult=true, interpSpeedDecreasing=25, interpSpeedIncreasing=25}
    boneToModify {boneName=dyn_pad_frt_r}
    rotation=<rot 6.427 14.408 -45>
   }
   Output Pose(Result=Component To Local(ComponentPose=@n23)) {layerGroup=None}
   Save cached pose 'AnimCache'(Pose=Copy Pose From Mesh() {bCopyCurves=true, bUseAttachedParent=true}) {cacheName=AnimCache}
   Sequence Player 'M_MED_DieselPunk_02_Arms_Only'() {playRate=0.01, sequence=&/M_MED_DieselPunk_02_Arms_Only}
  ]}
 EventGraph {nodes [
   n1 = Cast To FortPlayerPawn(Object=GetOwningActor())
   n2 = GetWeaponCoreAnim!(self:Fort Player Animation Interface Object Reference=GetAnimationInterface!(self:Fort Player Pawn Object Reference=@n1) -> (ReturnValue:Fort Player Animation Interface Object Reference)) -> (ReturnValue:EFortWeaponCoreAnimation Enum)
   > n3 = Set weapon_enum(weapon_enum:EFortWeaponCoreAnimation Enum=@n2) -> (Output_Get:EFortWeaponCoreAnimation Enum)
   > n4 = Branch(Condition="Equal (Enum)"(A:EFortWeaponCoreAnimation Enum=@n2, B:EFortWeaponCoreAnimation Enum=Pistol))
   > Set pistol_active(pistol_active:bool=true) -> (Output_Get:bool)
   > Set rifle_active(rifle_active:bool=false) -> (Output_Get:bool)
   > Set rl_active(rl_active:bool=false) -> (Output_Get:bool)
   > Set chaingun_active(chaingun_active:bool=false) -> (Output_Get:bool)
   Set pistol_active(execute=@n4.else, pistol_active=false)
   > n5 = Branch(Condition=OR Boolean!(
     A="Equal (Enum)"(A=@n2, B=Shotgun)
     B="Equal (Enum)"(A=@n2, B=Rifle)
     C:bool="Equal (Enum)"(A=@n2, B=AssaultRifle)
     D:bool="Equal (Enum)"(A=@n2, B=SniperRifle)
     E:bool="Equal (Enum)"(A=@n2, B=TacticalShotgun)
    ) -> (ReturnValue))
   > Set rifle_active(rifle_active=true)
   > Set rl_active(rl_active=false)
   > Set chaingun_active(chaingun_active=false)
   Set rifle_active(execute=@n5.else, rifle_active=false)
   > n6 = Branch(Condition="Equal (Enum)"(A=@n2, B=RocketLauncher))
   > Set rl_active(rl_active=true)
   > Set chaingun_active(chaingun_active=false)
   Set rl_active(execute=@n6.else, rl_active=false)
   > n7 = Branch(Condition="Equal (Enum)"(A=@n2, B=GoingCommando))
   > Set chaingun_active(chaingun_active=true)
   Set chaingun_active(execute=@n7.else, chaingun_active=false)
   n8 = Sequence!(execute) -> (then_0, then_1, then_2:Exec, then_3:Exec, then_4:Exec)
   > Set CrouchAlpha(CrouchAlpha:float="To Float (Boolean)"(InBool=Get bIsCrouching())) -> (Output_Get:float)
   n9 = MapRangeClamped(Value=Get PawnSpeed2D(), InRangeA=0, InRangeB=100)
   n10 = Branch(execute=@n8.then_3, Condition="float <= float"!(A:double=@n9, B:double=0) -> (ReturnValue:bool) {is=title K2Node_CallFunction})
   > Set SpeedMaskAlpha(SpeedMaskAlpha:float=0.05) -> (Output_Get:float)
   > Set SpeedMaskSpring(SpeedMaskSpring:float=1000) -> (Output_Get:float)
   Set SpeedMaskAlpha(execute=@n10.else, SpeedMaskAlpha=1)
   > Set SpeedMaskSpring(SpeedMaskSpring=200)
   n11 = GetOwnerSkeletalMeshComponent()
   n12 = BreakRotator(InRot=BreakTransform(InTransform=GetDeltaTransformFromRefPose(self=@n11, BoneName=upperarm_r)).Rotation)
   Set ArmPistolPadAlpha(execute=@n8.then_2, ArmPistolPadAlpha:float=MapRangeClamped(Value=@n12.Yaw, InRangeA=55, InRangeB=118)) -> (Output_Get:float)
   > n13 = Set ArmPistolChestAlpha(ArmPistolChestAlpha:float=MapRangeClamped(Value=@n12.Yaw, InRangeA=0, InRangeB=65)) -> (Output_Get:float)
   Set ArmPistolAlpha(execute=[@n14.else, @n13.then], ArmPistolAlpha:float=MapRangeClamped(Value=@n12.Yaw, InRangeA=5, InRangeB=118)) -> (Output_Get:float)
   Set skydive_active(execute=@n8.then_1, skydive_active:bool=Get bIsSkyDiving()) -> (Output_Get:bool)
   n14 = Branch(Condition=Get bIsTargeting())
   > Set ArmPistolAlpha(ArmPistolAlpha=1)
   PrintString!(
    execute
    InString=Enum to String!(Enumerator:EFortWeaponCoreAnimation Enum=@n3) -> (ReturnValue:String)
    bPrintToScreen
    bPrintToLog
    TextColor
    Duration
    Key
   ) -> (then)
   float * float(A=@n9, B=0.5)
   Event BlueprintUpdateAnimation()
  ]}
]
variables [
 chaingun_active {type=boolean}
 pistol_active {type=boolean}
 rifle_active {type=boolean}
 rl_active {type=boolean}
 skydive_active {type=boolean}
 weapon_enum {type=EFortWeaponCoreAnimation, default=Melee}
]
