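# Physics asset for the male medium DieselPunk 02 head cosmetic mesh, providing simple physics bodies for its dynamic hair.
# The head bone and two hair bones, dyn_hairA_01 and dyn_hairB_01, each carry a single capsule, with the hair capsules offset along their bone's Z axis.
# Each hair bone is constrained directly to the head with matching swing and twist angle limits, forming two independent one-link chains.
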PhysicsAsset /BRCosmetics/Characters/Player/Male/Medium/Bodies/M_MED_DieselPunk_02/Meshes/M_MED_Dieselpunk_Head_02_Physics
bodies {
 dyn_hairA_01 [{shapeType=Capsule, length=2.25, radius=2.25, center=<0 0 -1.2472>}]
 dyn_hairB_01 [{shapeType=Capsule, length=2.25, radius=2.25, center=<0 0 -2.0482>}]
 head [{shapeType=Capsule, length=5, radius=2}]
}
constraints [
 {bone1Name=dyn_hairA_01, bone2Name=head, swing1LimitDegrees=45, swing2LimitDegrees=45, twistLimitDegrees=45}
 {bone1Name=dyn_hairB_01, bone2Name=head, swing1LimitDegrees=45, swing2LimitDegrees=45, twistLimitDegrees=45}
]
